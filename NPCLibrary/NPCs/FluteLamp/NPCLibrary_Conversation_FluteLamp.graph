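# Conversation database for the FluteLamp NPC, a jumpy, Shaggy-style character that players talk to.
# One entry point leads to a speech node whose contextual lines cover first-ever meetings, repeat conversations in the match and specific player outfits.
# First-meeting and outfit lines take priority from named curve rows; repeat lines have none.
# The speech node then branches through link nodes into shared quest and basic-service conversations, plus two Season37 VoiceFork entries, one for delivery.
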ConversationDatabase /NPCLibrary/NPCs/FluteLamp/NPCLibrary_Conversation_FluteLamp
entry_points [{tag=Conversation.Entry.FluteLamp, destinations [#1]}]
nodes [
 ConversationEntryPointNode_7 {
  id=#1
  class=ConversationEntryPointNode
  outputs [#2]
  properties {entryTag {tagName=Conversation.Entry.FluteLamp}}
 }
 Speech_C_4 {
  id=#2
  class=Speech_C
  outputs [#3, #4, #5, #6]
  properties {
   nodeName=""
   generalConfig {
    defaultMessage="ZOINKS! You, like, can't sneak up on a guy like that!"
    contextualMessages [
     {
      message="(groan) This place is haunted too? BY DEMONS? What a bummer..."
      contextRequirements [{requirement {class=FortControllerRequirement_IsFirstEverConversationWithNPC}}]
      priority {curve {rowName=Tandem.Speech.Priority.FirstEverInteraction}}
     }
     {
      message="ZOINKS! You, like, can't sneak up on a guy like that!"
      contextRequirements [{requirement {
         class=FortControllerRequirement_CheckNPCMatchConversationCount
         properties {nPCUniqueIdTag {tagName=AISpawnerData.Type.Tandem.FluteLamp}, requiredConversationCountConfig {requirement=GreaterThan}}
        }}]
      priority {curve {curveTable=null, rowName=None}}
     }
     {
      message="I, like, don't think I packed enough snacks for this..."
      contextRequirements [{requirement {
         class=FortControllerRequirement_CheckNPCMatchConversationCount
         properties {nPCUniqueIdTag {tagName=AISpawnerData.Type.Tandem.FluteLamp}, requiredConversationCountConfig {requirement=GreaterThan}}
        }}]
      priority {curve {curveTable=null, rowName=None}}
     }
     {
      message=D-Demons? W-where?
      contextRequirements [{requirement {
         class=FortControllerRequirement_CheckNPCMatchConversationCount
         properties {nPCUniqueIdTag {tagName=AISpawnerData.Type.Tandem.FluteLamp}, requiredConversationCountConfig {requirement=GreaterThan}}
        }}]
      priority {curve {curveTable=null, rowName=None}}
     }
     {
      message="Hiya Scoob!"
      contextRequirements [{requirement {class=FortControllerRequirement_HasCID, properties {allowedCIds [/BRCosmetics/Athena/Items/Cosmetics/Characters/Character_LycheeNickel]}}}]
      priority {curve {rowName=Tandem.Speech.Priority.OutfitReaction}}
     }
     {
      message="Why can't we ever investigate, like, a sandwich place or something?"
      contextRequirements [{requirement {
         class=FortControllerRequirement_HasCID
         properties {allowedCIds [
           /BRCosmetics/Athena/Items/Cosmetics/Characters/Character_ArcanaAgate
           /BRCosmetics/Athena/Items/Cosmetics/Characters/Character_OboeThorn
           /BRCosmetics/Athena/Items/Cosmetics/Characters/Character_BaskIsle
          ]}
        }}]
      priority {curve {rowName=Tandem.Speech.Priority.OutfitReaction}}
     }
    ]
   }
  }
 }
 ConversationLinkNode_3 {
  id=#6
  class=ConversationLinkNode
  properties {remoteEntryTag {tagName=Conversation.Entry.Season37.VoiceFork.FluteLamp.Deliver}}
 }
 ConversationLinkNode_2 {
  id=#5
  class=ConversationLinkNode
  properties {remoteEntryTag {tagName=Conversation.Entry.Season37.VoiceFork.FluteLamp}}
 }
 ConversationLinkNode_0 {id=#4, class=ConversationLinkNode, properties {remoteEntryTag {tagName=Conversation.Entry.BasicServices}}}
 ConversationLinkNode_1 {id=#3, class=ConversationLinkNode, properties {remoteEntryTag {tagName=Conversation.Entry.NPCLibrary.Quests}}}
]
